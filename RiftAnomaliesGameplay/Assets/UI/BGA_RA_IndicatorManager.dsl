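; Indicator manager actor for Rift Anomalies that pings the closest anomaly for a player.
; A custom event has the given player's Athena marker component mark this actor on the server.
; It then waits for an initial ping duration read from a scalable float and destroys itself.
; A comment notes that spawned anomaly items handle player pinging themselves.
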(event Custom|MarkClosestAnomalyActorForPlayer
  (params (TargetPlayer :type Athena%20Marker%20Component%20Object%20Reference))
  (MarkerComponent|MarkActoronServer TargetPlayer self false true)
  ;[1 1 1] The spawned anomaly items themselves will handle player ping ability
  ; Remove the initial anomaly ping once it's cleared
  (Utilities|FlowControl|Delay (ScalableFloat|GetValueatLevel (Variables|Default|GetSF_InitialAnomalyPingDuration) 1.0))
  (Actor|DestroyActor))

(fn ConstructionScript)
